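# Sound cue for the Engaged Walk emote in Athena, layering two looping sounds for as long as it plays.
# A mixer combines the main Engaged Walk loop, started after a short fixed delay, with a much quieter typing loop that starts immediately.
# Despite its SFX name, it uses the emote music sound class and concurrency set, emote attenuation, and a send to the music audio-analysis submix.
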SoundCue /Game/Athena/Sounds/Emotes/Engaged_Walk/Emote_Engaged_Walk_SFX_Cue
root=Mixer_0
nodes [
 {id=Delay_0, children [WavePlayer_0], properties {delayMax=0.291, delayMin=0.291}}
 {id=Mixer_0, children [Delay_0,WavePlayer_1], properties {inputVolume [1,0.1]}}
 {id=WavePlayer_0, properties {bLooping=true, soundWaveAssetPtr=&/Emote_Engaged_Walk_01}}
 {id=WavePlayer_1, properties {bLooping=true, soundWaveAssetPtr=&/Emote_Engaged_Walk_Typing_Loop}}
]
properties {
 attenuationSettings=&Fort_Emotes_Athena_Att
 maxDistance=1600
 soundClassObject=&Fort_Music_Emote
 volumeMultiplier=1.25
 concurrencyOverrides {maxCount=3, resolutionRule=StopOldest}
 concurrencySet [/Game/Sounds/Concurrency_Rules/Athena_EmoteMusic_Conc]
 soundSubmixSends [{minSendLevel=0, soundSubmix=&Fortnite_AudioAnalysis_Music}]
}
